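# Post-process style Animation Blueprint for the F_MED_LintMermaid female character body, adding secondary motion to pouches and knives.
# The attached parent's pose passes through spine-weighted belly-knife constraints, a Control Rig, then AnimDynamics on three pouch bones.
# Additive bone offsets adjust both knives, using alphas derived from spine_03-to-pelvis rotation deltas, plus a blended skydiving offset.
# The mesh used for those deltas is cached when character customization loads.
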AnimBlueprint /BRCosmetics/Characters/Player/Female/Medium/Bodies/F_MED_LintMermaid/Meshes/F_MED_LintMermaid_AnimBP
target_skeleton=&/F_MED_LintMermaid_Skeleton
graphs [
 AnimGraph {nodes [
   n1 = "Constraint - dyn_belly_knife_1_l (PPP)"!(
    ComponentPose=Local To Component(LocalPose=Copy Pose From Mesh() {bUseAttachedParent=true})
    Alpha=1
    ConstraintWeights_0=0.474
    ConstraintWeights_1:float=0.502
    ConstraintWeights_2:float=0.024
   ) -> (Pose) {
    alphaBoolBlend {blendOption=Linear}
    boneToModify {boneName=dyn_belly_knife_1_l}
    constraintSetup [{targetBone {boneName=spine_02}}, {targetBone {boneName=spine_03}}, {targetBone {boneName=spine_04}}]
    constraintWeights [1,1,1]
   }
   n2 = "Constraint - dyn_belly_knife_2_l (PPP)"!(ComponentPose=@n1, Alpha=1, ConstraintWeights_0=0.479, ConstraintWeights_1:float=0.499, ConstraintWeights_2:float=0.022) -> (Pose) {
    alphaBoolBlend {blendOption=Linear}
    boneToModify {boneName=dyn_belly_knife_2_l}
    constraintSetup [{targetBone {boneName=spine_02}}, {targetBone {boneName=spine_03}}, {targetBone {boneName=spine_04}}]
    constraintWeights [1,1,1]
   }
   n3 = "Constraint - dyn_belly_knife_3_l (PPP)"!(ComponentPose=@n2, Alpha=1, ConstraintWeights_0=0.482, ConstraintWeights_1:float=0.499, ConstraintWeights_2:float=0.019) -> (Pose) {
    alphaBoolBlend {blendOption=Linear}
    boneToModify {boneName=dyn_belly_knife_3_l}
    constraintSetup [{targetBone {boneName=spine_02}}, {targetBone {boneName=spine_03}}, {targetBone {boneName=spine_04}}]
    constraintWeights [1,1,1]
   }
   n4 = "Constraint - dyn_belly_knife_1_r (PPP)"!(ComponentPose=@n3, Alpha=1, ConstraintWeights_0=0.475, ConstraintWeights_1:float=0.5, ConstraintWeights_2:float=0.025) -> (Pose) {
    alphaBoolBlend {blendOption=Linear}
    boneToModify {boneName=dyn_belly_knife_1_r}
    constraintSetup [{targetBone {boneName=spine_02}}, {targetBone {boneName=spine_03}}, {targetBone {boneName=spine_04}}]
    constraintWeights [1,1,1]
   }
   n5 = "Constraint - dyn_belly_knife_2_r (PPP)"!(ComponentPose=@n4, Alpha=1, ConstraintWeights_0=0.482, ConstraintWeights_1:float=0.499, ConstraintWeights_2:float=0.019) -> (Pose) {
    alphaBoolBlend {blendOption=Linear}
    boneToModify {boneName=dyn_belly_knife_2_r}
    constraintSetup [{targetBone {boneName=spine_02}}, {targetBone {boneName=spine_03}}, {targetBone {boneName=spine_04}}]
    constraintWeights [1,1,1]
   }
   n6 = "Constraint - dyn_belly_knife_3_r (PPP)"!(ComponentPose=@n5, Alpha=1, ConstraintWeights_0=0.479, ConstraintWeights_1:float=0.499, ConstraintWeights_2:float=0.022) -> (Pose) {
    alphaBoolBlend {blendOption=Linear}
    boneToModify {boneName=dyn_belly_knife_3_r}
    constraintSetup [{targetBone {boneName=spine_02}}, {targetBone {boneName=spine_03}}, {targetBone {boneName=spine_04}}]
    constraintWeights [1,1,1]
   }
   n7 = Control Rig(Source=Component To Local(ComponentPose=@n6)) {
    alphaBoolBlend {blendOption=Linear}
    controlRigAssetReference {blueprintRigClass=&/F_MED_LintMermaid_CtrlRig.F_MED_LintMermaid_CtrlRig_C}
   }
   n8 = Anim Dynamics - Bone: dyn_pouch_l(ComponentPose=Local To Component(LocalPose=@n7), Alpha=1, ExternalForce="0,0,0") {
    angularDampingOverride=0.999
    bOverrideAngularDamping=true
    simulationSpace=RootRelative
    alphaBoolBlend {blendOption=Linear}
    boundBone {boneName=dyn_pouch_l}
    componentAppliedLinearAccClamp=<1000 1200 2500>
    componentLinearAccScale=<0.3 0.2 0.5>
    componentLinearVelScale=<0.4 0.3 0.6>
    physicsBodyDefinitions [{
      boundBone {boneName=dyn_pouch_l}
      boxExtents=<10 8 10>
      constraintSetup {angularLimitsMax=<0 3 0>, angularLimitsMin=<0 -3 -30>, angularTarget=<0 0 0>}
      localJointOffset=<-0.4 0 0.08>
     }]
   }
   n9 = Anim Dynamics - Bone: dyn_pouch_r(ComponentPose=@n8, Alpha=1, ExternalForce="0,0,0") {
    angularDampingOverride=0.999
    angularSpringConstant=500
    bOverrideAngularDamping=true
    simulationSpace=RootRelative
    alphaBoolBlend {blendOption=Linear}
    boundBone {boneName=dyn_pouch_r}
    componentAppliedLinearAccClamp=<1000 1200 2500>
    componentLinearAccScale=<0.3 0.2 0.5>
    componentLinearVelScale=<0.4 0.3 0.6>
    gravityOverride {z=-980}
    physicsBodyDefinitions [{
      boundBone {boneName=dyn_pouch_r}
      boxExtents=<10 8 10>
      constraintSetup {angularLimitsMax=<0 3 0>, angularLimitsMin=<0 -3 -30>, angularTarget=<0 0 0>}
      localJointOffset=<0.4 0 -0.08>
     }]
    relativeSpaceBone {boneName=pelvis}
   }
   n10 = Anim Dynamics - Bone: dyn_big_pouch_r(ComponentPose=@n9, Alpha=1, ExternalForce="0,0,0") {
    angularDampingOverride=0.999
    bOverrideAngularDamping=true
    bUseGravityOverride=true
    simulationSpace=RootRelative
    alphaBoolBlend {blendOption=Linear}
    boundBone {boneName=dyn_big_pouch_r}
    componentAppliedLinearAccClamp=<1000 1200 2500>
    componentLinearAccScale=<0.3 0.2 0.5>
    componentLinearVelScale=<0.4 0.3 0.6>
    gravityOverride {z=-980}
    physicsBodyDefinitions [{
      boundBone {boneName=dyn_big_pouch_r}
      boxExtents=<10 10 10>
      constraintSetup {angularLimitsMax=<0 3 0>, angularLimitsMin=<0 -3 -15>, angularTarget=<0 0 0>}
      localJointOffset=<0.5 0 0.08>
     }]
   }
   n11 = "Transform (Modify) Bone - Bone: dyn_knife_l"(
    ComponentPose=@n10
    Alpha=Get Left Knife Body Twist Left Alpha() -> (Left Knife Body Twist Left Alpha:double)
    Translation="0,0.97,-0.954"
    Rotation="-6,20,12.879"
    Scale="1,1,1"
   ) {
    rotationMode=BMM_Additive
    rotationSpace=BCS_BoneSpace
    translationMode=BMM_Additive
    translationSpace=BCS_BoneSpace
    alphaBoolBlend {blendOption=Linear}
    boneToModify {boneName=dyn_knife_l}
    rotation=<rot -6 20 12.879>
    translation {y=0.97, z=-0.954}
   }
   n12 = "Transform (Modify) Bone - Bone: dyn_knife_l"(
    ComponentPose=@n11
    Alpha=Get Left Knife Body Twist Right Alpha() -> (Left Knife Body Twist Right Alpha:double)
    Translation="0.018,0.273,0"
    Rotation="2.812,4.001,0.032"
    Scale="1,1,1"
   ) {
    rotationMode=BMM_Additive
    rotationSpace=BCS_BoneSpace
    translationMode=BMM_Additive
    translationSpace=BCS_BoneSpace
    alphaBoolBlend {blendOption=Linear}
    boneToModify {boneName=dyn_knife_l}
    rotation=<rot 2.8119 4.001 0.032>
    translation=<0.018 0.273>
   }
   n13 = "Transform (Modify) Bone - Bone: dyn_knife_r"(
    ComponentPose=@n12
    Alpha=Get Right Knife Body Twist Right Alpha() -> (Right Knife Body Twist Right Alpha:double)
    Translation="-0.056,-0.078,0.044"
    Rotation="7.729,7,14.076"
    Scale="1,1,1"
   ) {
    rotationMode=BMM_Additive
    rotationSpace=BCS_BoneSpace
    translationMode=BMM_Additive
    translationSpace=BCS_BoneSpace
    alphaBoolBlend {blendOption=Linear}
    boneToModify {boneName=dyn_knife_r}
    rotation=<rot 7.729 7 14.076>
    translation=<-0.056 -0.078 0.044>
   }
   n14 = Get bIsSkyDiving()
   n15 = "Transform (Modify) Bone - Bone: dyn_knife_r"!(ComponentPose=@n13, bAlphaBoolEnabled:bool=@n14, Translation="-0.363,-0.468,-0.644", Rotation="-21.798,5.280,-28.905", Scale="1,1,1") -> (Pose) {
    alphaInputType=Bool
    rotationMode=BMM_Additive
    rotationSpace=BCS_BoneSpace
    translationMode=BMM_Additive
    translationSpace=BCS_BoneSpace
    alphaBoolBlend {blendInTime=0.3, blendOption=Linear, blendOutTime=0.3}
    boneToModify {boneName=dyn_knife_r}
    rotation=<rot -21.798 5.28 -28.905>
    translation=<-0.363 -0.468 -0.644>
   }
   n16 = "Transform (Modify) Bone - Bone: dyn_knife_l"!(ComponentPose=@n15, bAlphaBoolEnabled:bool=@n14, Translation="0.082,0.767,0.432", Rotation="-32.047,14.511,-35.248", Scale="1,1,1") -> (Pose) {
    alphaInputType=Bool
    rotationMode=BMM_Additive
    rotationSpace=BCS_BoneSpace
    translationMode=BMM_Additive
    translationSpace=BCS_BoneSpace
    alphaBoolBlend {blendInTime=0.3, blendOption=Linear, blendOutTime=0.3}
    boneToModify {boneName=dyn_knife_l}
    rotation=<rot -32.047 14.511 -35.248>
    translation=<0.082 0.767 0.432>
   }
   Output Pose(Result=Component To Local(ComponentPose=@n16)) {layerGroup=None}
  ]}
 EventGraph {nodes [
   Event BlueprintUpdateAnimation()
   > n1 = Set Left Knife Body Twist Right Alpha(Left Knife Body Twist Right Alpha:double=GetAlphaFromReferenceTransformDelta(
     SkeletalMeshComponent=Get Skeletal Mesh() -> (Skeletal Mesh:Skeletal Mesh Component Object Reference)
     BoneName=spine_03
     BaseName=pelvis
     DeltaComponentType=EulerX
     InputRange=(X=-20,Y=0)
     OutputRange=(X=1,Y=0)
     TwistAxis="1,0,0"
    )) -> (Output_Get:double)
   n2 = GetAlphaFromReferenceTransformDelta(
    SkeletalMeshComponent=Get Skeletal Mesh()
    BoneName=spine_03
    BaseName=pelvis
    DeltaComponentType=EulerY
    InputRange=(X=-16,Y=-5)
    OutputRange=(X=0.8,Y=0)
    TwistAxis="1,0,0"
   )
   n3 = Event OnCharacterCustomizationLoaded()
   > n4 = Set Skeletal Mesh(Skeletal Mesh:Skeletal Mesh Component Object Reference=@n3.CustomBaseMesh) -> (Output_Get:Skeletal Mesh Component Object Reference)
   n5 = float - float(
    A=GetAlphaFromReferenceTransformDelta(SkeletalMeshComponent=@n4, BoneName=spine_03, BaseName=pelvis, DeltaComponentType=EulerX, InputRange=(X=-3,Y=36), TwistAxis="1,0,0")
    B=GetAlphaFromReferenceTransformDelta(
     SkeletalMeshComponent=Get Skeletal Mesh()
     BoneName=spine_03
     BaseName=pelvis
     InputRange=(X=4,Y=20)
     OutputRange=(X=0.5,Y=0)
     TwistAxis="1,0,0"
    )
   )
   Set Left Knife Body Twist Left Alpha(execute=@n1.then, Left Knife Body Twist Left Alpha:double="Clamp (Float)"(Value=float + float(A=@n5, B=@n2), Min=0, Max=1)) -> (Output_Get:double)
   n6 = Get Skeletal Mesh()
   > Set Right Knife Body Twist Right Alpha(Right Knife Body Twist Right Alpha:double=float + float(
     A=@n2
     B=GetAlphaFromReferenceTransformDelta(
      SkeletalMeshComponent=@n6
      BoneName=spine_03
      BaseName=pelvis
      DeltaComponentType=EulerX
      InputRange=(X=-25,Y=0)
      OutputRange=(X=1,Y=0)
      TwistAxis="1,0,0"
     )
    )) -> (Output_Get:double)
  ]}
]
variables [
 Left Knife Body Twist Left Alpha {type=number}
 Left Knife Body Twist Right Alpha {type=number}
 Right Knife Body Twist Right Alpha {type=number}
 Skeletal Mesh {type=SkeletalMeshComponent}
]
