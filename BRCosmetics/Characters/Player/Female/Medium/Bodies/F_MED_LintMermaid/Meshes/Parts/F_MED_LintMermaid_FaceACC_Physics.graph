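# Physics asset for the face accessory mesh of the female medium LintMermaid player cosmetic, set up for physics on a hanging braid.
# A five-bone dyn_braid chain hangs from the head, with its base locked in swing and twist and a capsule on each link below it.
# Those links allow tight twist and swing that loosens toward the tip.
# Simple spheres and capsules on head, neck, upper spine and clavicles provide body collision shapes.
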PhysicsAsset /BRCosmetics/Characters/Player/Female/Medium/Bodies/F_MED_LintMermaid/Meshes/Parts/F_MED_LintMermaid_FaceACC_Physics
bodies {
 clavicle_l [{shapeType=Capsule, length=20, radius=7.5, shapeName=clavicle_l_capsule, center=<15.6016 0.0839 -2.8259>, rotation=<rot 0 -90 90>}]
 clavicle_r [{shapeType=Capsule, length=20, radius=7.5, shapeName=clavicle_r_capsule, center=<-15.6016 -0.0839 2.8259>, rotation=<rot 0 -90 90>}]
 dyn_braid_01 [{
   shapeType=Capsule
   length=5.2956
   radius=2.1664
   shapeName=dyn_braid_01_capsule
   center=<3.464 1.3133 0.0552>
   rotation=<rot -0.8435 -89.178 -87.9631>
  }]
 dyn_braid_02 [{
   shapeType=Capsule
   length=4.0562
   radius=2.2181
   shapeName=dyn_braid_02_capsule
   center=<3.6902 0.697 -0.4556>
   rotation=<rot -2.465 -91.2195 -83.0897>
  }]
 dyn_braid_03 [{
   shapeType=Capsule
   length=4.2412
   radius=1.3492
   shapeName=dyn_braid_03_capsule
   center=<2.8502 -0.3969 0.0738>
   rotation=<rot 0 87.4528 -87.7042>
  }]
 dyn_braid_04 [{
   shapeType=Capsule
   length=4.3354
   radius=1.8521
   shapeName=dyn_braid_04_capsule
   center=<3.7748 0.0217 0.1287>
   rotation=<rot 0 88.6415 -86.764>
  }]
 dyn_braid_base [{shapeType=Sphere, radius=1, shapeName=dyn_braid_base_sphere}]
 head [{shapeType=Sphere, radius=12.3351, shapeName=head_sphere, center=<6.5928 3.3202 0.0032>}]
 neck_01 [{shapeType=Capsule, length=0, radius=6.505, shapeName=neck_01_capsule, center=<3.3529 1.6803 0>, rotation=<rot 0 -90 0>}]
 spine_05 [{shapeType=Capsule, length=14.5, radius=10.505, shapeName=spine_05_capsule, center=<0.1827 0.7811 0>, rotation=<rot 0 -90 0>}]
}
constraints [
 {bone1Name=dyn_braid_01, bone2Name=dyn_braid_base, swing1LimitDegrees=30, swing2LimitDegrees=20, twistLimitDegrees=3}
 {bone1Name=dyn_braid_02, bone2Name=dyn_braid_01, swing1LimitDegrees=40, swing2LimitDegrees=30, twistLimitDegrees=3}
 {bone1Name=dyn_braid_03, bone2Name=dyn_braid_02, swing1LimitDegrees=40, swing2LimitDegrees=30, twistLimitDegrees=3}
 {bone1Name=dyn_braid_04, bone2Name=dyn_braid_03, swing1LimitDegrees=40, swing2LimitDegrees=40, twistLimitDegrees=3}
 {
  bone1Name=dyn_braid_base
  bone2Name=head
  swing1LimitDegrees=45
  swing1Motion=Locked
  swing2LimitDegrees=45
  swing2Motion=Locked
  twistLimitDegrees=45
  twistMotion=Locked
 }
]
