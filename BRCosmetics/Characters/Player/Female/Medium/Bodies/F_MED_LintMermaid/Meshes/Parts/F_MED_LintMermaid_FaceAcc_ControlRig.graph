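# Control Rig for the LintMermaid female face accessory that procedurally poses a hair braid as the head moves.
# The first two braid bones aim at the top spline bone, then receive a transform offset.
# Four spherical pose readers on the head, relative to neck or spine, weight corrective offsets on the braid spline bones and braid.
# A Hermite spline built from those spline bones then drives the lower three braid bones through a stretched chain fit.
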ControlRigBlueprint /BRCosmetics/Characters/Player/Female/Medium/Bodies/F_MED_LintMermaid/Meshes/Parts/F_MED_LintMermaid_FaceAcc_ControlRig
graphs [
 RigVMFunctionLibrary {links [], nodes []}
 RigVMModel {nodes [
   RigUnit_BeginExecution = &RigUnit_BeginExecution()
   /Script/ControlRig.RigUnit_AimItem(
    DebugSettings=(bEnabled=False,Scale=10,WorldOffset=(Rotation=(X=0,Y=0,Z=0,W=1),Translation=(X=0,Y=0,Z=0),Scale3D=(X=1,Y=1,Z=1)))
    Item=(Type=Bone,Name="dyn_braid_01")
    Primary=(Weight=1,Axis=(X=1,Y=0,Z=0),Target=(X=0,Y=0,Z=-1),Kind=Location,Space=(Type=Bone,Name="dyn_braid_spline_00"))
    Secondary=(Weight=0,Axis=(X=0,Y=0,Z=1),Target=(X=1,Y=0,Z=0),Kind=Location,Space=(Type=Bone,Name="None"))
    Weight=1
    ExecutePin=@RigUnit_BeginExecution.ExecutePin
   ) {is=struct RigVMUnitNode}
   > &RigUnit_AimItem(
    DebugSettings=(bEnabled=False,Scale=10,WorldOffset=(Rotation=(X=0,Y=0,Z=0,W=1),Translation=(X=0,Y=0,Z=0),Scale3D=(X=1,Y=1,Z=1)))
    Item=(Type=Bone,Name="dyn_braid_02")
    Primary=(Weight=1,Axis=(X=1,Y=0,Z=0),Target=(X=0,Y=0,Z=-1),Kind=Location,Space=(Type=Bone,Name="dyn_braid_spline_00"))
    Secondary=(Weight=0,Axis=(X=0,Y=0,Z=1),Target=(X=1,Y=0,Z=0),Kind=Location,Space=(Type=Bone,Name="None"))
    Weight=1
   )
   > &RigUnit_ModifyTransforms(
    ItemToModify="((Item=(Type=Bone,Name=\"dyn_braid_02\"),Transform=(Rotation=(X=0,Y=0,Z=-0.1305,W=0.9914),Translation=(X=0,Y=0.6282,Z=0),Scale3D=(X=1,Y=1,Z=1))),(Item=(Type=Bone,Name=\"dyn_braid_01\"),Transform=(Rotation=(X=0,Y=0,Z=0.0245,W=0.9997),Translation=(X=-0.004,Y=1.1 ...(truncated)"
    Weight=1
    WeightMinimum=0
   )
   > n1 = &RigUnit_SphericalPoseReader(
    ActiveRegionScaleFactors=(PositiveWidth=2,NegativeWidth=0.8,PositiveHeight=2,NegativeHeight=2)
    ActiveRegionSize=0.1
    Debug=(bDrawDebug=false,bDraw2D=false,bDrawLocalAxes=false,DebugScale=25,DebugSegments=20,DebugThickness=0.25)
    DriverAxis=(X=1,Y=0,Z=0)
    DriverItem=(Type=Bone,Name="head")
    FalloffRegionScaleFactors=(PositiveWidth=1,NegativeWidth=0.7,PositiveHeight=2.6,NegativeHeight=2.6)
    FalloffSize=0.3
    FlipHeightScaling=False
    FlipWidthScaling=False
    OptionalParentItem=(Type=Bone,Name="neck_01")
    RotationOffset=(X=90,Y=0,Z=-130)
   )
   > &RigUnit_ModifyTransforms(
    ItemToModify="((Item=(Type=Bone,Name=\"dyn_braid_spline_00\"),Transform=(Rotation=(X=0,Y=0,Z=0,W=1),Translation=(X=0,Y=-1.8901,Z=-4.1987),Scale3D=(X=1,Y=1,Z=1))),(Item=(Type=Bone,Name=\"dyn_braid_spline_01\"),Transform=(Rotation=(X=0,Y=0,Z=0,W=1),Translation=(X=0,Y=-1.6062, ...(truncated)"
    WeightMinimum=0
    Weight=@n1.OutputParam
   )
   > n2 = &RigUnit_SphericalPoseReader(
    ActiveRegionScaleFactors=(PositiveWidth=2,NegativeWidth=0.8,PositiveHeight=2,NegativeHeight=2)
    ActiveRegionSize=0.1
    Debug=(bDrawDebug=false,bDraw2D=false,bDrawLocalAxes=false,DebugScale=25,DebugSegments=20,DebugThickness=0.25)
    DriverAxis=(X=0,Y=1,Z=0)
    DriverItem=(Type=Bone,Name="head")
    FalloffRegionScaleFactors=(PositiveWidth=1,NegativeWidth=0.7,PositiveHeight=2.2,NegativeHeight=2)
    FalloffSize=0.3
    FlipHeightScaling=false
    FlipWidthScaling=true
    OptionalParentItem=(Type=Bone,Name="spine_04")
    RotationOffset=(X=90,Y=0,Z=39)
   )
   > &RigUnit_ModifyTransforms(
    ItemToModify="((Item=(Type=Bone,Name=\"dyn_braid_spline_00\"),Transform=(Rotation=(X=0,Y=0,Z=0,W=1),Translation=(X=1.3528,Y=1.2238,Z=2.6136),Scale3D=(X=1,Y=1,Z=1))),(Item=(Type=Bone,Name=\"dyn_braid_spline_01\"),Transform=(Rotation=(X=0,Y=0,Z=0,W=1),Translation=(X=0.0875,Y= ...(truncated)"
    WeightMinimum=0
    Weight=@n2.OutputParam
   )
   > n3 = &RigUnit_SphericalPoseReader(
    ActiveRegionScaleFactors=(PositiveWidth=2,NegativeWidth=2,PositiveHeight=1,NegativeHeight=1)
    ActiveRegionSize=0.1
    Debug=(bDrawDebug=false,bDraw2D=false,bDrawLocalAxes=false,DebugScale=25,DebugSegments=20,DebugThickness=0.25)
    DriverAxis=(X=0,Y=1,Z=0)
    DriverItem=(Type=Bone,Name="head")
    FalloffRegionScaleFactors=(PositiveWidth=1,NegativeWidth=1.2,PositiveHeight=1,NegativeHeight=1)
    FalloffSize=0.3
    FlipHeightScaling=False
    FlipWidthScaling=False
    OptionalParentItem=(Type=Bone,Name="spine_04")
    RotationOffset=(X=158,Y=-5,Z=-5)
   )
   > &RigUnit_ModifyTransforms(
    ItemToModify="((Item=(Type=Bone,Name=\"dyn_braid_spline_00\"),Transform=(Rotation=(X=0,Y=0,Z=0,W=1),Translation=(X=-0.8483,Y=0,Z=1.3042),Scale3D=(X=1,Y=1,Z=1))),(Item=(Type=Bone,Name=\"dyn_braid_spline_01\"),Transform=(Rotation=(X=0,Y=0,Z=0,W=1),Translation=(X=0.5318,Y=0,Z= ...(truncated)"
    WeightMinimum=0
    Weight=@n3.OutputParam
   )
   > n4 = &RigUnit_SphericalPoseReader(
    ActiveRegionScaleFactors=(PositiveWidth=1,NegativeWidth=1,PositiveHeight=2,NegativeHeight=2)
    ActiveRegionSize=0.1
    Debug=(bDrawDebug=false,bDraw2D=false,bDrawLocalAxes=false,DebugScale=25,DebugSegments=20,DebugThickness=0.25)
    DriverAxis=(X=1,Y=0,Z=0)
    DriverItem=(Type=Bone,Name="head")
    FalloffRegionScaleFactors=(PositiveWidth=0.5,NegativeWidth=1.2,PositiveHeight=2,NegativeHeight=2)
    FalloffSize=0.3
    FlipHeightScaling=False
    FlipWidthScaling=False
    OptionalParentItem=(Type=Bone,Name="spine_04")
    RotationOffset=(X=0,Y=-50,Z=0)
   )
   > &RigUnit_ModifyTransforms(
    ItemToModify="((Item=(Type=Bone,Name=\"dyn_braid_spline_00\"),Transform=(Rotation=(X=0,Y=0,Z=0,W=1),Translation=(X=1.7337,Y=0,Z=0),Scale3D=(X=1,Y=1,Z=1))),(Item=(Type=Bone,Name=\"dyn_braid_01\"),Transform=(Rotation=(X=0,Y=0.0175,Z=0,W=0.9998),Translation=(X=0,Y=0,Z=0),Scale ...(truncated)"
    WeightMinimum=0
    Weight=@n4.OutputParam
   )
   > SplineFromItems = RigVMFunctionReferenceNode(
    Items=((Type=Bone,Name="dyn_braid_spline_00"),(Type=Bone,Name="dyn_braid_spline_01"),(Type=Bone,Name="dyn_braid_spline_02"),(Type=Bone,Name="dyn_braid_spline_03"))
    Samples Per Segment=16
    Spline Mode=Hermite
   )
   Make_Array = RigVMDispatchNode(Values=((Type=Bone,Name="dyn_braid_03"),(Type=Bone,Name="dyn_braid_04"),((),Type=Bone,Name="dyn_braid_05"))) {template="DISPATCH_RigVMDispatch_ArrayMake(in Values,out Array)"}
   > /Script/ControlRigSpline.RigUnit_FitChainToSplineCurveItemArray(
    Alignment=Stretched
    DebugSettings=(bEnabled=False,Scale=1,CurveColor=(R=1,G=1,B=0,A=1),SegmentsColor=(R=1,G=0,B=0,A=1),WorldOffset=(Rotation=(X=0,Y=0,Z=0,W=1),Translation=(X=0,Y=0,Z=0),Scale3D=(X=1,Y=1,Z=1)))
    Maximum=1
    Minimum=0
    PoleVectorPosition=(X=0,Y=0,Z=0)
    PrimaryAxis=(X=1,Y=0,Z=0)
    RotationEaseType=Linear
    Rotations=()
    SamplingPrecision=12
    SecondaryAxis=(X=0,Y=0,Z=0)
    Weight=1
    bPropagateToChildren=True
    Items=@Make_Array.Array
    Spline=@SplineFromItems.Spline
   ) {is=struct RigVMUnitNode}
  ]}
]
hierarchy [
 root {loc=<0 0 0>, rot=<0 0 0 1>}
 pelvis {parent=root, loc=<0 -0.5301 98.0571>, rot=<0.0127 0.707 -0.0127 -0.707>}
 spine_01 {parent=pelvis, loc=<6.1208 -0.9328 0>, rot=<0 0 0.034 -0.9994>}
 spine_02 {parent=spine_01, loc=<6.5997 -0.2095 0>, rot=<0 0 0.0238 -0.9997>}
 spine_03 {parent=spine_02, loc=<6.4914 -0.1237 0>, rot=<0 0 0.0149 -0.9999>}
 spine_04 {parent=spine_03, loc=<6.9598 0.0794 0>, rot=<0 0 -0.0132 -0.9999>}
 spine_05 {parent=spine_04, loc=<6.9987 0 0>, rot=<0 0 -0.0015 -1>}
 neck_01 {parent=spine_05, loc=<11.3138 0.1529 0>, rot=<0 0 -0.2003 -0.9797>}
 neck_02 {parent=neck_01, loc=<4.5627 0 0>, rot=<0 0 0.0504 -0.9987>}
 head {parent=neck_02, loc=<5.3914 0 0>, rot=<0 0 0.115 -0.9934>}
 dyn_braid_base {parent=head, loc=<12.0531 -8.9968 0>, rot=<0 0 0.9942 -0.1073>}
 dyn_braid_01 {parent=dyn_braid_base, loc=<0 0 0>, rot=<0 0 0 1>}
 dyn_braid_02 {parent=dyn_braid_01, loc=<7.3388 0 0>, rot=<0 0 0.1853 -0.9827>}
 dyn_braid_03 {parent=dyn_braid_02, loc=<7.2592 0 0>, rot=<0 0 -0.2807 -0.9598>}
 dyn_braid_04 {parent=dyn_braid_03, loc=<6.4349 0 0>, rot=<0 0 -0.1314 -0.9913>}
 dyn_braid_05 {parent=dyn_braid_04, loc=<5.8914 0 0>, rot=<0 0 0 1>}
 dyn_strand_base_r {parent=head, loc=<3.2723 2.1249 6.432>, rot=<0 0 -0.0822 -0.9966>}
 dyn_strand_01_r {parent=dyn_strand_base_r, loc=<0 0 0>, rot=<0 0 0 1>}
 dyn_strand_02_r {parent=dyn_strand_01_r, loc=<-4.268 -0.0001 0>, rot=<0 0 0 1>}
 dyn_strand_base_l {parent=head, loc=<3.2727 2.1249 -6.432>, rot=<0 0 0.9966 -0.0822>}
 dyn_strand_01_l {parent=dyn_strand_base_l, loc=<0 0 0>, rot=<0 0 0 1>}
 dyn_strand_02_l {parent=dyn_strand_01_l, loc=<4.2689 0 0>, rot=<0 0 0 1>}
 dyn_braid_spline_00 {parent=neck_02, loc=<1.1709 -8.244 0.005>, rot=<-0.0784 -0.7027 0.0784 -0.7027>}
 dyn_braid_spline_01 {parent=neck_01, loc=<2.2511 -7.5245 0.005>, rot=<-0.1137 -0.6979 0.1137 -0.6979>}
 clavicle_l {parent=spine_05, loc=<5.4831 -0.1053 -3.2946>, rot=<0.0465 -0.7289 -0.009 -0.6829>}
 clavicle_r {parent=spine_05, loc=<5.4831 -0.1053 3.2946>, rot=<0.7289 0.0465 -0.6829 0.009>}
 dyn_braid_spline_02 {parent=spine_05, loc=<13.3368 -6.8028 0.004>, rot=<0.0284 -0.7065 -0.0284 -0.7065>}
 dyn_braid_spline_03 {parent=spine_05, loc=<8.4532 -8.2784 0.005>, rot=<0.0284 -0.7065 -0.0284 -0.7065>}
]
